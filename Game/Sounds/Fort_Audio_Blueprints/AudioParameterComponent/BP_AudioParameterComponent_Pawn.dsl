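; Audio parameter component for the local player pawn target that feeds water interaction state to sound.
; Entering or exiting water binds or unbinds footstep and movement-mode delegates and refreshes depth immediately.
; Updates call SetWaterDepths, then write a WaterDepth integer and IsCrouching boolean to the pawn's audio parameter component.
; Helpers provide a smoothed, normalized immersion depth and debug-gated parameter printing.
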(event EventBeginPlay)

(fn GetWaterDepthValue
  (bind _output_get (Variables|Water|SetWaterDepthValue (Math|Interpolation|FInterpTo (Variables|Water|GetWaterDepthValue) (Math|Float|SelectFloat (WaterBody|GetWaterImmersionDepth (Variables|Default|GetLocalPlayerPawnTarget)) 0.0 (!= (WaterBody|GetWaterImmersionDepth (Variables|Default|GetLocalPlayerPawnTarget)) -1.0)) (Utilities|Time|GetWorldDeltaSeconds) 5e+01)))
  (return (Math|Float|MapRangeClamped _output_get 0.0 1.3e+02 0.0 1.0)))

(fn MovementModeChange
  (params (Character :type Character%20Object%20Reference) (PrevMovementMode :type EMovementMode%20Enum :default "MOVE_None") (PreviousCustomMode :type Byte))
  (WaterInteraction|OnWaterUpdate 0))

(fn OnEnterWater
  (params (FortPlayerPawn :type Fort%20Player%20Pawn%20Object%20Reference))
  (Variables|Water|SetIsInWater true)
  (EventDispatchers|BindEventtoOnFootstepEvent (Variables|Default|GetLocalPlayerPawnTarget) (EventDispatchers|CreateEvent :!event OnWaterUpdate))
  (Character|BindEventtoMovementModeChangedDelegate (Variables|Default|GetLocalPlayerPawnTarget) (EventDispatchers|CreateEvent :!event Movement%20Mode%20Change))
  (WaterInteraction|OnWaterUpdate 0))

(fn OnExitWater
  (params (FortPlayerPawn :type Fort%20Player%20Pawn%20Object%20Reference) (bPawnIsOutsideOfAllWaterBodies :type Boolean))
  (Variables|Water|SetIsInWater false)
  (EventDispatchers|UnbindEventfromOnFootstepEvent (Variables|Default|GetLocalPlayerPawnTarget) (EventDispatchers|CreateEvent :!event OnWaterUpdate))
  (Character|UnbindEventfromMovementModeChangedDelegate (Variables|Default|GetLocalPlayerPawnTarget) (EventDispatchers|CreateEvent :!event Movement%20Mode%20Change))
  (WaterInteraction|OnWaterUpdate 0))

(fn OnWaterDepthSwitch
  (bind _localplayerpawntarget (Variables|Default|GetLocalPlayerPawnTarget)
    (:then
      (Audio|Parameter|SetIntegerParameter (AudioParameter|GetAudioParameterComponent _localplayerpawntarget :!class /Script/FortniteGame.FortPawn) "WaterDepth" (Utilities|Select :Index (Variables|Water|GetPlayerWaterDepth) :NewEnumerator5 0 :NewEnumerator4 1 :NewEnumerator3 2 :NewEnumerator2 3 :NewEnumerator1 4) :!class /Script/AudioExtensions.AudioParameterControllerInterface)
      (Audio|Parameter|SetBooleanParameter (AudioParameter|GetAudioParameterComponent _localplayerpawntarget :!class /Script/FortniteGame.FortPawn) "IsCrouching" (AI|Components|NavMovement|IsCrouching (Variables|/Script/Engine.Character|GetCharacterMovement (Variables|Default|GetLocalPlayerPawnTarget))) :!class /Script/AudioExtensions.AudioParameterControllerInterface))
    (:else)))

(fn OnWaterUpdate
  (params (FootIndex :type Integer))
  (Default|SetWaterDepths 0.55 0.25 0.01
    (:Out
      (WaterInteraction|OnWaterDepthSwitch))))

(fn PrintParameter
  (params (Parameter :type Audio%20Parameter%20Query%20Data%20Structure%20%28by%20ref%29) (Condition :type Boolean))
  (bind (_tagquery_12_74b441d54446a49eb7f00aac728b68f0 _parametername_2_0e82f2ef4dd09c9eb9e25d9ad7db520d _controlbus_5_0b33e9404596de8d58698ea40f70c171) (Utilities|Struct|BreakAudioParameterQueryData Parameter))
  (if (Variables|Default|GetbDebugEnabled)
    (Development|PrintText (Utilities|Text|FormatText "NSLOCTEXT(\"[F665B7AAE8CEE542F4D312EE9CBBB24F]\", \"82DDFE9E4B3EC5BBB66DE789F475C79F\", \"Parameter CB {CB} Name {Name} Value {Value}\")" (Utilities|GetDisplayName _controlbus_5_0b33e9404596de8d58698ea40f70c171) _parametername_2_0e82f2ef4dd09c9eb9e25d9ad7db520d Condition))))
